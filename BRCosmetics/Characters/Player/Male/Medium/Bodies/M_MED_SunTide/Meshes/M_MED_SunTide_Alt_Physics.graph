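# Physics asset for the alternate SunTide male medium body cosmetic, covering its dangling belt, left-thigh chain and right-thigh can.
# Simple sphere and capsule bodies on pelvis, lower spine and thighs form the base, joined by free-swinging, free-twisting constraints.
# A four-link belt hangs from a pelvis master bone with wide swing and minimal twist.
# The two-link chain, ending in a three-capsule body, and the can have tighter limits off their thighs.
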PhysicsAsset /BRCosmetics/Characters/Player/Male/Medium/Bodies/M_MED_SunTide/Meshes/M_MED_SunTide_Alt_Physics
bodies {
 dyn_belt_01 [{shapeType=Capsule, length=5.5, radius=3, shapeName=dyn_belt_01_capsule, center=<2 0 0>, rotation=<rot 90 0 0>}]
 dyn_belt_02 [{shapeType=Capsule, length=5.5, radius=3, shapeName=dyn_belt_02_capsule, center=<2 0 0>, rotation=<rot 90 0 0>}]
 dyn_belt_03 [{shapeType=Capsule, length=5.5, radius=3, shapeName=dyn_belt_02_capsule, center=<2 0 0>, rotation=<rot 90 0 0>}]
 dyn_belt_04 [{shapeType=Capsule, length=5.5, radius=4, shapeName=dyn_belt_04_capsule, center=<2 0 0>, rotation=<rot 90 0 0>}]
 dyn_belt_mstr [{shapeType=Capsule, length=0.505, radius=0.505, shapeName=dyn_belt_mstr_capsule, rotation=<rot 0 -90 0>}]
 dyn_can [{shapeType=Capsule, length=11, radius=3.7669, shapeName=dyn_can_capsule, center=<5.1345 0.9594 -3.8398>, rotation=<rot -1.863 -90 -5.0795>}]
 dyn_chain_l_01 [{shapeType=Capsule, length=3.755, radius=3.255, shapeName=dyn_chain_l_01_capsule, center=<5.1427 0 0>, rotation=<rot 90 0 0>}]
 dyn_chain_l_02 [
  {shapeType=Capsule, length=6.5, radius=6.7309, shapeName=dyn_chain_l_02_capsule, center=<4.1697 0 3.7207>, rotation=<rot 43.6882 0 -0.0004>}
  {
   shapeType=Capsule
   length=6.25
   radius=3.75
   shapeName=dyn_chain_l_02_capsule
   center=<-1.1754 7.6746 2.2474>
   rotation=<rot 33.7682 46.6641 -74.5385>
  }
  {
   shapeType=Capsule
   length=6.25
   radius=3.75
   shapeName=dyn_chain_l_02_capsule
   center=<-1.7654 -8.8531 1.2742>
   rotation=<rot 38.0716 -51.6487 -103.9301>
  }
 ]
 pelvis [
  {shapeType=Sphere, radius=11.5, shapeName=pelvis_sphere, center=<-18.0575 0.1411 -5.4269>}
  {shapeType=Capsule, length=14, radius=8, shapeName=pelvis_capsule, center=<-18.1865 3.2301 0>}
 ]
 spine_01 [{
   shapeType=Capsule
   length=11.5697
   radius=8.726
   shapeName=spine_01_capsule
   center=<0.6389 -2.8271 -0.0204>
   rotation=<rot -0.4758 -90 -0.2782>
  }]
 thigh_l [{shapeType=Capsule, length=20, radius=8.5, shapeName=thigh_l_capsule, center=<-16.8506 0 0>, rotation=<rot 90 0 0>}]
 thigh_r [{shapeType=Capsule, length=21.65, radius=9.05, shapeName=thigh_r_capsule, center=<13.4898 0.1824 0>, rotation=<rot 90 180 179.6767>}]
 thigh_twist_01_l [{
   shapeType=Capsule
   length=9.4434
   radius=5
   shapeName=thigh_twist_01_l_capsule
   center=<6.0791 0 3.2525>
   rotation=<rot 29.6148 179.9999 -177.7035>
  }]
}
constraints [
 {
  bone1Name=spine_01
  bone2Name=pelvis
  swing1LimitDegrees=45
  swing1Motion=Free
  swing2LimitDegrees=45
  swing2Motion=Free
  twistLimitDegrees=45
  twistMotion=Free
 }
 {
  bone1Name=thigh_l
  bone2Name=pelvis
  swing1LimitDegrees=45
  swing1Motion=Free
  swing2LimitDegrees=45
  swing2Motion=Free
  twistLimitDegrees=45
  twistMotion=Free
 }
 {
  bone1Name=thigh_twist_01_l
  bone2Name=thigh_l
  swing1LimitDegrees=45
  swing1Motion=Free
  swing2LimitDegrees=45
  swing2Motion=Free
  twistLimitDegrees=45
  twistMotion=Free
 }
 {bone1Name=dyn_chain_l_01, bone2Name=thigh_l, swing1LimitDegrees=1, swing2LimitDegrees=10, twistLimitDegrees=1}
 {bone1Name=dyn_chain_l_02, bone2Name=dyn_chain_l_01, swing1LimitDegrees=1, swing2LimitDegrees=15, twistLimitDegrees=1}
 {
  bone1Name=thigh_r
  bone2Name=pelvis
  swing1LimitDegrees=45
  swing1Motion=Free
  swing2LimitDegrees=45
  swing2Motion=Free
  twistLimitDegrees=45
  twistMotion=Free
 }
 {bone1Name=dyn_can, bone2Name=thigh_r, swing1LimitDegrees=20, swing2LimitDegrees=20, twistLimitDegrees=20}
 {
  bone1Name=dyn_belt_mstr
  bone2Name=pelvis
  swing1LimitDegrees=45
  swing1Motion=Free
  swing2LimitDegrees=45
  swing2Motion=Free
  twistLimitDegrees=45
  twistMotion=Free
 }
 {bone1Name=dyn_belt_01, bone2Name=dyn_belt_mstr, swing1LimitDegrees=65, swing2LimitDegrees=65, twistLimitDegrees=2}
 {bone1Name=dyn_belt_02, bone2Name=dyn_belt_01, swing1LimitDegrees=60, swing2LimitDegrees=60, twistLimitDegrees=2}
 {bone1Name=dyn_belt_03, bone2Name=dyn_belt_02, swing1LimitDegrees=60, swing2LimitDegrees=60, twistLimitDegrees=2}
 {bone1Name=dyn_belt_04, bone2Name=dyn_belt_03, swing1LimitDegrees=60, swing2LimitDegrees=60, twistLimitDegrees=2}
]
