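# Physics asset for the necklace mesh of the M_MED_SunTide male player body cosmetic, defining its dangling-chain physics and collision.
# A box on the necklace base and capsules on the two following dyn_necklace bones form a chain constrained to spine_05.
# Spine_05 carries two capsules and a rotated box as an upper-torso volume.
# All joints leave swing and twist free, so their stored angular limits do not apply.
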PhysicsAsset /BRCosmetics/Characters/Player/Male/Medium/Bodies/M_MED_SunTide/Meshes/M_MED_SunTide_Necklace_Physics
bodies {
 dyn_necklace_01 [{shapeType=Capsule, length=14.7587, radius=2.1258, shapeName=dyn_necklace_01_capsule, center=<6.3964 0.7787 0>}]
 dyn_necklace_02 [{shapeType=Capsule, length=15, radius=1.6515, shapeName=dyn_necklace_02_capsule, center=<4.0864 0 0>}]
 dyn_necklace_base [{shapeType=Box, extentX=4.9519, extentY=2.3, extentZ=17.0095, shapeName=dyn_necklace_parent_box, center=<5.8433 1.8392 0.3322>}]
 spine_05 [
  {shapeType=Capsule, length=28.4532, radius=10.3475, shapeName=spine_05_capsule, center=<-3.2211 0.2766 0>}
  {shapeType=Capsule, length=31.9298, radius=6.871, shapeName=spine_05_capsule, center=<6.3001 -2.4116 0.0001>}
  {shapeType=Box, extentX=22.8749, extentY=7.5, extentZ=32.1281, shapeName=spine_05_box, center=<-0.6602 5.6419 0>, rotation=<rot 0 -34.0383 0>}
 ]
}
constraints [
 {
  bone1Name=dyn_necklace_base
  bone2Name=spine_05
  swing1LimitDegrees=45
  swing1Motion=Free
  swing2LimitDegrees=45
  swing2Motion=Free
  twistLimitDegrees=45
  twistMotion=Free
 }
 {
  bone1Name=dyn_necklace_01
  bone2Name=dyn_necklace_base
  swing1LimitDegrees=45
  swing1Motion=Free
  swing2LimitDegrees=45
  swing2Motion=Free
  twistLimitDegrees=45
  twistMotion=Free
 }
 {
  bone1Name=dyn_necklace_02
  bone2Name=dyn_necklace_01
  swing1LimitDegrees=45
  swing1Motion=Free
  swing2LimitDegrees=45
  swing2Motion=Free
  twistLimitDegrees=45
  twistMotion=Free
 }
]
