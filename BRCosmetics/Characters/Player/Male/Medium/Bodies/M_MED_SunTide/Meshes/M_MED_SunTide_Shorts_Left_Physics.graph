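# Physics asset for the left leg of the SunTide medium male character's shorts, apparently for secondary cloth-like motion.
# Capsules on the thigh and calf bones represent the leg, alongside front, back, inner and outer shorts bones.
# Each shorts bone has tight swing and twist limits against its own tiny MSTR anchor body.
# All four anchors connect to thigh_twist_01_l with free angular constraints.
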PhysicsAsset /BRCosmetics/Characters/Player/Male/Medium/Bodies/M_MED_SunTide/Meshes/M_MED_SunTide_Shorts_Left_Physics
bodies {
 calf_l [{shapeType=Capsule, length=30.1468, radius=6.8983, center=<-19.8303 -3.3101 -0.4527>, rotation=<rot 90 180 -0.5532>}]
 calf_twist_02_l [{shapeType=Capsule, length=17.5381, radius=5.9619, center=<10.9556 -2.9077 -1.8471>, rotation=<rot 90 90 90.0002>}]
 dyn_shorts_bk_l [{shapeType=Capsule, length=20, radius=6, shapeName=dyn_shorts_bk_l_capsule, center=<20 0 1>, rotation=<rot 0 0 90>}]
 dyn_shorts_bk_l_MSTR [{shapeType=Capsule, length=1, radius=1}]
 dyn_shorts_ft_l [{shapeType=Capsule, length=18.9253, radius=6, shapeName=dyn_shorts_ft_l_capsule, center=<8.0241 1.8941 3>, rotation=<rot 0 0 90>}]
 dyn_shorts_ft_l_MSTR [{shapeType=Capsule, length=1, radius=1}]
 dyn_shorts_in_l [{shapeType=Capsule, length=10, radius=6, shapeName=dyn_shorts_in_l_capsule, center=<15.1315 0 3>, rotation=<rot 0 0 90>}]
 dyn_shorts_in_l_MSTR [{shapeType=Capsule, length=1, radius=1}]
 dyn_shorts_out_l [{shapeType=Capsule, length=12.1818, radius=6, shapeName=dyn_shorts_out_l_capsule, center=<15.2579 0 3>, rotation=<rot 0 0 90>}]
 dyn_shorts_out_l_MSTR [{shapeType=Capsule, length=1, radius=1}]
 thigh_twist_01_l [{shapeType=Capsule, length=29.3252, radius=8.3874, center=<-10.7326 -2.2111 -0.3562>, rotation=<rot 90 180 -5.1456>}]
}
constraints [
 {
  bone1Name=calf_twist_02_l
  bone2Name=calf_l
  swing1LimitDegrees=45
  swing1Motion=Free
  swing2LimitDegrees=45
  swing2Motion=Free
  twistLimitDegrees=45
  twistMotion=Free
 }
 {bone1Name=dyn_shorts_bk_l, bone2Name=dyn_shorts_bk_l_MSTR, swing1LimitDegrees=15, swing2LimitDegrees=15, twistLimitDegrees=5}
 {bone1Name=dyn_shorts_out_l, bone2Name=dyn_shorts_out_l_MSTR, swing1LimitDegrees=15, swing2LimitDegrees=15, twistLimitDegrees=5}
 {
  bone1Name=dyn_shorts_bk_l_MSTR
  bone2Name=thigh_twist_01_l
  swing1LimitDegrees=45
  swing1Motion=Free
  swing2LimitDegrees=45
  swing2Motion=Free
  twistLimitDegrees=45
  twistMotion=Free
 }
 {
  bone1Name=dyn_shorts_out_l_MSTR
  bone2Name=thigh_twist_01_l
  swing1LimitDegrees=45
  swing1Motion=Free
  swing2LimitDegrees=45
  swing2Motion=Free
  twistLimitDegrees=45
  twistMotion=Free
 }
 {
  bone1Name=dyn_shorts_ft_l_MSTR
  bone2Name=thigh_twist_01_l
  swing1LimitDegrees=45
  swing1Motion=Free
  swing2LimitDegrees=45
  swing2Motion=Free
  twistLimitDegrees=45
  twistMotion=Free
 }
 {bone1Name=dyn_shorts_ft_l, bone2Name=dyn_shorts_ft_l_MSTR, swing1LimitDegrees=15, swing2LimitDegrees=15, twistLimitDegrees=5}
 {
  bone1Name=dyn_shorts_in_l_MSTR
  bone2Name=thigh_twist_01_l
  swing1LimitDegrees=45
  swing1Motion=Free
  swing2LimitDegrees=45
  swing2Motion=Free
  twistLimitDegrees=45
  twistMotion=Free
 }
 {bone1Name=dyn_shorts_in_l, bone2Name=dyn_shorts_in_l_MSTR, swing1LimitDegrees=15, swing2LimitDegrees=15, twistLimitDegrees=5}
]
