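# Physics asset for the right leg of the SunTide male medium body's shorts, providing capsule bodies for secondary cloth-like bone motion.
# Four dynamic panel bones (front, back, inner, outer) each connect with narrow swing and twist limits to a tiny MSTR anchor body.
# The anchors attach to the upper thigh twist body with free angular motion, and that body and the calf attach likewise to the thigh.
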PhysicsAsset /BRCosmetics/Characters/Player/Male/Medium/Bodies/M_MED_SunTide/Meshes/M_MED_SunTide_Shorts_Right_Physics
bodies {
 calf_r [{shapeType=Capsule, length=25.4855, radius=7.7593, center=<14.9322 2.4647 0.7585>, rotation=<rot 90 -178.8251 -179.2001>}]
 dyn_shorts_bk_r [{shapeType=Capsule, length=20, radius=6, shapeName=dyn_shorts_bk_l_capsule, center=<20 0 1>, rotation=<rot 0 0 90>}]
 dyn_shorts_bk_r_MSTR [{shapeType=Capsule, length=1, radius=1}]
 dyn_shorts_ft_r [{shapeType=Capsule, length=20, radius=6, shapeName=dyn_shorts_ft_l_capsule, center=<8.0241 0 5>, rotation=<rot 0 0 90>}]
 dyn_shorts_ft_r_MSTR [{shapeType=Capsule, length=1, radius=1}]
 dyn_shorts_in_r [{shapeType=Capsule, length=10, radius=6, shapeName=dyn_shorts_in_l_capsule, center=<15.7195 0 3>, rotation=<rot 0 0 90>}]
 dyn_shorts_in_r_MSTR [{shapeType=Capsule, length=1, radius=1}]
 dyn_shorts_out_r [{shapeType=Capsule, length=10, radius=6, shapeName=dyn_shorts_out_l_capsule, center=<15.2579 0 3>, rotation=<rot 0 0 90>}]
 dyn_shorts_out_r_MSTR [{shapeType=Capsule, length=1, radius=1}]
 thigh_r [{shapeType=Capsule, length=28.8733, radius=7.1405, shapeName=thigh_r_capsule, center=<25.9852 0 0.5345>, rotation=<rot 85 179.9998 179.9999>}]
 thigh_twist_01_r [{shapeType=Capsule, length=30.1143, radius=8.5713, center=<6.0221 0.8886 1.1979>, rotation=<rot 90 0 0>}]
}
constraints [
 {
  bone1Name=dyn_shorts_in_r_MSTR
  bone2Name=thigh_twist_01_r
  swing1LimitDegrees=45
  swing1Motion=Free
  swing2LimitDegrees=45
  swing2Motion=Free
  twistLimitDegrees=45
  twistMotion=Free
 }
 {bone1Name=dyn_shorts_ft_r, bone2Name=dyn_shorts_ft_r_MSTR, swing1LimitDegrees=15, swing2LimitDegrees=15, twistLimitDegrees=2}
 {
  bone1Name=dyn_shorts_ft_r_MSTR
  bone2Name=thigh_twist_01_r
  swing1LimitDegrees=45
  swing1Motion=Free
  swing2LimitDegrees=45
  swing2Motion=Free
  twistLimitDegrees=45
  twistMotion=Free
 }
 {bone1Name=dyn_shorts_out_r, bone2Name=dyn_shorts_out_r_MSTR, swing1LimitDegrees=15, swing2LimitDegrees=15, twistLimitDegrees=2}
 {
  bone1Name=dyn_shorts_out_r_MSTR
  bone2Name=thigh_twist_01_r
  swing1LimitDegrees=45
  swing1Motion=Free
  swing2LimitDegrees=45
  swing2Motion=Free
  twistLimitDegrees=45
  twistMotion=Free
 }
 {bone1Name=dyn_shorts_bk_r, bone2Name=dyn_shorts_bk_r_MSTR, swing1LimitDegrees=15, swing2LimitDegrees=15, twistLimitDegrees=2}
 {
  bone1Name=dyn_shorts_bk_r_MSTR
  bone2Name=thigh_twist_01_r
  swing1LimitDegrees=45
  swing1Motion=Free
  swing2LimitDegrees=45
  swing2Motion=Free
  twistLimitDegrees=45
  twistMotion=Free
 }
 {bone1Name=dyn_shorts_in_r, bone2Name=dyn_shorts_in_r_MSTR, swing1LimitDegrees=15, swing2LimitDegrees=15, twistLimitDegrees=2}
 {
  bone1Name=calf_r
  bone2Name=thigh_r
  swing1LimitDegrees=45
  swing1Motion=Free
  swing2LimitDegrees=45
  swing2Motion=Free
  twistLimitDegrees=45
  twistMotion=Free
 }
 {
  bone1Name=thigh_twist_01_r
  bone2Name=thigh_r
  swing1LimitDegrees=45
  swing1Motion=Free
  swing2LimitDegrees=45
  swing2Motion=Free
  twistLimitDegrees=45
  twistMotion=Free
 }
]
