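# Physics asset for the dangling chain on the M_MED_SunTide male medium body cosmetic, covering the chain bones and the right thigh.
# Capsules cover the master bone, first three links and thigh_r, with a second capsule on link 03 and a box on the end link.
# Constraints run from thigh_r through dyn_chain_MSTR to dyn_chain_04.
# The thigh joint swings and twists freely, while link joints are angularly limited, loosest at the tip.
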PhysicsAsset /BRCosmetics/Characters/Player/Male/Medium/Bodies/M_MED_SunTide/Meshes/M_MED_SunTide_chain_Physic
bodies {
 dyn_chain_01 [{shapeType=Capsule, length=1.1943, radius=1.3057, shapeName=dyn_chain_01_capsule, center=<3 0 2>, rotation=<rot 0 0 90>}]
 dyn_chain_02 [{shapeType=Capsule, length=11.494, radius=2.4237, shapeName=dyn_chain_02_capsule, center=<3 0 2.7182>, rotation=<rot 0 0 90>}]
 dyn_chain_03 [
  {shapeType=Capsule, length=11.494, radius=2.4237, shapeName=dyn_chain_02_capsule, center=<3 0 2>, rotation=<rot 0 0 90>}
  {
   shapeType=Capsule
   length=13.7139
   radius=2.5823
   shapeName=dyn_chain_02_capsule
   center=<1.4678 9.3531 -7.208>
   rotation=<rot -9.8462 1.7536 -5.151>
  }
 ]
 dyn_chain_04 [{shapeType=Box, extentX=9.2943, extentY=3.0788, extentZ=7.059, shapeName=dyn_chain_04_box, center=<3.7915 0 0.1431>, rotation=<rot 0 0 90>}]
 dyn_chain_MSTR [{shapeType=Capsule, length=1, radius=1, shapeName=dyn_chain_MSTR_capsule}]
 thigh_r [{
   shapeType=Capsule
   length=27.2283
   radius=8.7855
   shapeName=thigh_r_capsule
   center=<18.4781 -0.2035 0.0433>
   rotation=<rot 85 179.9996 174.9998>
  }]
}
constraints [
 {bone1Name=dyn_chain_02, bone2Name=dyn_chain_01, swing1LimitDegrees=5, swing2LimitDegrees=25, twistLimitDegrees=1}
 {bone1Name=dyn_chain_03, bone2Name=dyn_chain_02, swing1LimitDegrees=5, swing2LimitDegrees=25, twistLimitDegrees=1}
 {bone1Name=dyn_chain_04, bone2Name=dyn_chain_03, swing1LimitDegrees=45, swing2LimitDegrees=45, twistLimitDegrees=5}
 {bone1Name=dyn_chain_01, bone2Name=dyn_chain_MSTR, swing1LimitDegrees=15, swing2LimitDegrees=15, twistLimitDegrees=5}
 {
  bone1Name=dyn_chain_MSTR
  bone2Name=thigh_r
  swing1LimitDegrees=45
  swing1Motion=Free
  swing2LimitDegrees=45
  swing2Motion=Free
  twistLimitDegrees=45
  twistMotion=Free
 }
]
