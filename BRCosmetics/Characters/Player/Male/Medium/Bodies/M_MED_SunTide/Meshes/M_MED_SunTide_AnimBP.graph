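# Animation Blueprint for the SunTide male player body cosmetic, adding secondary motion to clothing and accessories over a pose copied from the attached parent mesh.
# Deform and belt Control Rigs feed RigidBody simulations for shorts and chains.
# A long series of constraints, Look Ats, bone offsets, AnimDynamics and Trail controllers then adjusts sleeves, necklace, bracelet, can and tassels.
# A further Control Rig reads crouch, tactical sprint, jump and skydive states.
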AnimBlueprint /BRCosmetics/Characters/Player/Male/Medium/Bodies/M_MED_SunTide/Meshes/M_MED_SunTide_AnimBP
target_skeleton=&/M_MED_SunTide_Skeleton
graphs [
 AnimGraph {nodes [
   n1 = Anim Dynamics - Bone: dyn_chain_02!(ComponentPose, Alpha=1, GravityOverride:Vector="0,0,-980", ExternalForce="0,0,0") -> (Pose) {
    angularDampingOverride=0.9999
    angularSpringConstant=500
    bAngularSpring=true
    bGravityOverrideInSimSpace=true
    bOverrideAngularDamping=true
    bUseGravityOverride=true
    bUseSphericalLimits=true
    simulationSpace=BoneRelative
    alphaBoolBlend {blendOption=Linear}
    boundBone {boneName=dyn_chain_02}
    componentLinearAccScale=<0.2 1 1>
    componentLinearVelScale=<0.2 1 1>
    gravityOverride {z=-980}
    physicsBodyDefinitions [{
      sphereCollisionRadius=0
      boundBone {boneName=dyn_chain_02}
      boxExtents=<15 30 30>
      constraintSetup {angularLimitsMax=<5 5 15>, angularLimitsMin=<-1 -35 -15>, angularTarget=<0 0 0>}
      localJointOffset=<2 0 0>
     }]
    relativeSpaceBone {boneName=root}
    sphericalLimits [{limitRadius=10, drivingBone {boneName=thigh_r}, sphereLocalOffset=<3 0 0>}]
   }
   n2 = Anim Dynamics - Bone: dyn_chain_03!(ComponentPose=@n1, Alpha=1, GravityOverride:Vector="0,0,-980", ExternalForce="0,0,0") -> (Pose) {
    angularDampingOverride=0.9999
    angularSpringConstant=500
    bAngularSpring=true
    bGravityOverrideInSimSpace=true
    bOverrideAngularDamping=true
    bUseGravityOverride=true
    bUseSphericalLimits=true
    simulationSpace=BoneRelative
    alphaBoolBlend {blendOption=Linear}
    boundBone {boneName=dyn_chain_03}
    componentLinearAccScale=<0.2 1 1>
    componentLinearVelScale=<0.2 1 1>
    gravityOverride {z=-980}
    physicsBodyDefinitions [{
      sphereCollisionRadius=0
      boundBone {boneName=dyn_chain_03}
      boxExtents=<15 30 30>
      constraintSetup {angularLimitsMax=<5 5 15>, angularLimitsMin=<-1 -45 -15>, angularTarget=<0 0 0>}
      localJointOffset=<8 0 0>
     }]
    relativeSpaceBone {boneName=root}
    sphericalLimits [{limitRadius=10, drivingBone {boneName=thigh_r}, sphereLocalOffset=<10 0 0>}]
   }
   Anim Dynamics - Bone: dyn_chain_04!(ComponentPose=@n2, Alpha=1, GravityOverride:Vector="0,0,-980", ExternalForce="0,0,0") -> (Pose) {
    angularDampingOverride=0.9999
    angularSpringConstant=500
    bAngularSpring=true
    bGravityOverrideInSimSpace=true
    bOverrideAngularDamping=true
    bUseGravityOverride=true
    bUseSphericalLimits=true
    simulationSpace=BoneRelative
    alphaBoolBlend {blendOption=Linear}
    boundBone {boneName=dyn_chain_04}
    componentLinearAccScale=<0.2 1 1>
    componentLinearVelScale=<0.2 1 1>
    gravityOverride {z=-980}
    physicsBodyDefinitions [{
      sphereCollisionRadius=0
      boundBone {boneName=dyn_chain_04}
      boxExtents=<15 30 30>
      constraintSetup {angularLimitsMax=<5 3 15>, angularLimitsMin=<-1 -45 -15>, angularTarget=<0 0 0>}
      localJointOffset=<15 0 0>
     }]
    relativeSpaceBone {boneName=root}
    sphericalLimits [{limitRadius=10, drivingBone {boneName=thigh_r}, sphereLocalOffset=<12 0 0>}]
   }
   Anim Dynamics - Bone: dyn_can(Alpha=0.5, ExternalForce="0,0,0") {
    angularDampingOverride=1
    angularSpringConstant=1500
    bGravityOverrideInSimSpace=true
    bOverrideAngularDamping=true
    gravityScale=0.15
    simulationSpace=BoneRelative
    alphaBoolBlend {blendOption=Linear}
    boundBone {boneName=dyn_can}
    physicsBodyDefinitions [{
      sphereCollisionRadius=0
      boundBone {boneName=dyn_can}
      boxExtents=<10 20 20>
      constraintSetup {angularLimitsMax=<0 0 2>, angularLimitsMin=<0 0 -15>, angularTarget=<0 0 0>}
      localJointOffset=<5 0 0>
     }]
    relativeSpaceBone {boneName=pelvis}
   }
   Constraint!(ComponentPose, Alpha=1) -> (Pose) {alphaBoolBlend {blendOption=Linear}, boneToModify=~}
   "Transform (Modify) Bone - Bone: dyn_belt_03"!(ComponentPose, Alpha=1, Translation="16.219,3.092,1.335", Rotation="24.574,1.868,-3.846") -> (Pose) {
    rotationMode=BMM_Replace
    rotationSpace=BCS_BoneSpace
    scaleMode=BMM_Replace
    scaleSpace=BCS_BoneSpace
    translationMode=BMM_Additive
    translationSpace=BCS_BoneSpace
    alphaBoolBlend {blendOption=Linear}
    alphaScaleBiasClamp {bInterpResult=true, interpSpeedDecreasing=5, interpSpeedIncreasing=7}
    boneToModify {boneName=dyn_belt_03}
    rotation=<rot 24.574 1.868 -3.846>
    translation=<16.219 3.092 1.335>
   }
   "Transform (Modify) Bone - Bone: dyn_chain_04"!(ComponentPose, Alpha=1, Translation="0,0,0", Rotation="5,0,0") -> (Pose) {
    rotationMode=BMM_Replace
    rotationSpace=BCS_BoneSpace
    scaleMode=BMM_Replace
    scaleSpace=BCS_BoneSpace
    translationMode=BMM_Additive
    translationSpace=BCS_BoneSpace
    alphaBoolBlend {blendOption=Linear}
    alphaScaleBiasClamp {bInterpResult=true, interpSpeedDecreasing=5, interpSpeedIncreasing=7}
    boneToModify {boneName=dyn_chain_04}
    rotation {pitch=5}
   }
   "Transform (Modify) Bone - Bone: dyn_belt_mstr"!(ComponentPose, Alpha=1, Translation="3.5,-0.5,2.5", Rotation="0,0,0") -> (Pose) {
    rotationMode=BMM_Replace
    rotationSpace=BCS_BoneSpace
    scaleSpace=BCS_BoneSpace
    translationMode=BMM_Replace
    translationSpace=BCS_BoneSpace
    alphaBoolBlend {blendOption=Linear}
    alphaScaleBiasClamp {bInterpResult=true, interpSpeedDecreasing=20, interpSpeedIncreasing=7}
    boneToModify {boneName=dyn_belt_mstr}
    scale {x=0.5}
    translation=<3.5 -0.5 2.5>
   }
   RigidBody!(ComponentPose, Alpha, OverrideWorldGravity:Vector="0,0,-980", ExternalForce) -> (Pose) {
    bClampLinearTranslationLimitToRefPose=true
    bOverrideWorldGravity=true
    overridePhysicsAsset=&/M_MED_SunTide_Necklace_Physics
    simulationSpace=BaseBoneSpace
    alphaBoolBlend {blendOption=Linear}
    baseBoneRef {boneName=spine_05}
    componentAppliedLinearAccClamp=<6000 6000 6000>
    componentLinearAccScale=<0.25 0.25 0.25>
    componentLinearVelScale=<0.25 0.25 0.25>
    overrideWorldGravity {z=-980}
   }
   n3 = Control Rig(Source=Copy Pose From Mesh() {bUseAttachedParent=true}) {
    alphaBoolBlend {blendOption=Linear}
    controlRigAssetReference {blueprintRigClass=&/M_MED_SunTide_DeformRig_Master_ControlRig_new.M_MED_SunTide_DeformRig_Master_ControlRig_new_C}
   }
   n4 = Control Rig!(Source=@n3, Alpha, Emote:double=Get Emote() -> (Emote:double)) -> (Pose) {
    alphaBoolBlend {blendOption=Linear}
    controlRigAssetReference {blueprintRigClass=&/M_MED_SunTide_Alt_Belt_CtrlRig.M_MED_SunTide_Alt_Belt_CtrlRig_C}
   }
   n5 = RigidBody(ComponentPose=Local To Component(LocalPose=@n4), OverrideWorldGravity="0,0,-980") {
    bClampLinearTranslationLimitToRefPose=true
    bOverrideWorldGravity=true
    overridePhysicsAsset=&/M_MED_SunTide_Shorts_Left_Physics
    simulationSpace=BaseBoneSpace
    alphaBoolBlend {blendOption=Linear}
    baseBoneRef {boneName=calf_l}
    componentAppliedLinearAccClamp=<1000 1000 1000>
    componentLinearAccScale=<0.2 0.2 0.2>
    componentLinearVelScale=<0.2 0.2 0.2>
    overrideWorldGravity {z=-980}
   }
   n6 = RigidBody(ComponentPose=@n5, OverrideWorldGravity="0,0,-980") {
    bClampLinearTranslationLimitToRefPose=true
    bOverrideWorldGravity=true
    overridePhysicsAsset=&/M_MED_SunTide_Shorts_Right_Physics
    simulationSpace=BaseBoneSpace
    alphaBoolBlend {blendOption=Linear}
    baseBoneRef {boneName=calf_r}
    componentAppliedLinearAccClamp=<1000 1000 1000>
    componentLinearAccScale=<0.2 0.2 0.2>
    componentLinearVelScale=<0.2 0.2 0.2>
    overrideWorldGravity {z=-980}
   }
   n7 = RigidBody(ComponentPose=@n6, OverrideWorldGravity="0,0,-980") {
    bClampLinearTranslationLimitToRefPose=true
    bOverrideWorldGravity=true
    overridePhysicsAsset=&/M_MED_SunTide_chain_Physic
    simulationSpace=BaseBoneSpace
    alphaBoolBlend {blendOption=Linear}
    baseBoneRef {boneName=pelvis}
    componentAppliedLinearAccClamp=<3000 3000 3000>
    componentLinearAccScale=<0.4 0.3 0.3>
    componentLinearVelScale=<0.4 0.3 0.3>
    overrideWorldGravity {z=-980}
    simSpaceSettings {
     maxAngularAcceleration=500
     maxAngularVelocity=3
     maxLinearAcceleration=400
     maxLinearVelocity=250
     velocityScaleZ=0
     worldAlpha=1
     externalLinearDragV=<0.1 0.1>
    }
   }
   n8 = RigidBody!(ComponentPose=@n7, Alpha, ExternalForce) -> (Pose) {
    bClampLinearTranslationLimitToRefPose=true
    bOverrideWorldGravity=true
    overridePhysicsAsset=&/M_MED_SunTide_Alt_Physics
    simulationSpace=BaseBoneSpace
    alphaBoolBlend {blendOption=Linear}
    baseBoneRef {boneName=pelvis}
    componentAppliedLinearAccClamp=<3000 3000 3000>
    componentLinearAccScale=<0.3 0.3 0.5>
    componentLinearVelScale=<0.3 0.3 0.5>
    overrideWorldGravity {z=-980}
   }
   n9 = "Constraint - dyn_belt_02 (PPP)"!(ComponentPose=@n8, Alpha=1, ConstraintWeights_0=0.2603, ConstraintWeights_1:float=0.5555, ConstraintWeights_2:float=0.1842) -> (Pose) {
    alphaBoolBlend {blendOption=Linear}
    alphaScaleBiasClamp {bClampResult=true, bInterpResult=true, clampMax=0.45, interpSpeedDecreasing=20, interpSpeedIncreasing=20}
    boneToModify {boneName=dyn_belt_02}
    constraintSetup [{targetBone {boneName=pelvis}}, {targetBone {boneName=thigh_twist_01_l}}, {targetBone {boneName=deform_groin_l}}]
    constraintWeights [1,1,1]
   }
   n10 = "Constraint - dyn_belt_03 (PPPPP)"!(
    ComponentPose=@n9
    Alpha=1
    ConstraintWeights_0=0.0904
    ConstraintWeights_1:float=0.0932
    ConstraintWeights_2:float=0
    ConstraintWeights_3:float=0.8013
    ConstraintWeights_4:float=0.0152
   ) -> (Pose) {
    alphaBoolBlend {blendOption=Linear}
    alphaScaleBiasClamp {bClampResult=true, bInterpResult=true, clampMax=0.6, interpSpeedDecreasing=20, interpSpeedIncreasing=20}
    boneToModify {boneName=dyn_belt_03}
    constraintSetup [
     {targetBone {boneName=pelvis}}
     {targetBone {boneName=thigh_l}}
     {targetBone {boneName=calf_l}}
     {targetBone {boneName=thigh_twist_01_l}}
     {targetBone {boneName=deform_groin_l}}
    ]
    constraintWeights [1,1,1,1,1]
   }
   n11 = "Transform (Modify) Bone - Bone: dyn_chain_l_02"(ComponentPose=@n10, Alpha=0, Translation="-1.241,0.000,-1.000", Rotation="-10,0.000,31.917", Scale="1,1,1") {
    rotationMode=BMM_Additive
    translationMode=BMM_Additive
    alphaBoolBlend {blendOption=Linear}
    boneToModify {boneName=dyn_chain_l_02}
    rotation {pitch=-10, roll=31.917}
    translation {x=-1.241, z=-1}
   }
   n12 = Look At - Bone: dyn_sleeve_fr_l(ComponentPose=@n11, Alpha=1) {
    alphaBoolBlend {blendOption=Linear}
    boneToModify {boneName=dyn_sleeve_fr_l}
    lookAtLocation=<20.3856 9.0295 5.806>
    lookAtTarget {boneReference {boneName=upperarm_l}}
    lookAt_Axis {axis=<1 0>}
   }
   n13 = Look At - Bone: dyn_sleeve_bk_l(ComponentPose=@n12, Alpha=1) {
    alphaBoolBlend {blendOption=Linear}
    boneToModify {boneName=dyn_sleeve_bk_l}
    lookAtLocation=<19.5059 -10.8585 -0.3306>
    lookAtTarget {boneReference {boneName=upperarm_l}}
    lookAt_Axis {axis=<1 0>}
   }
   n14 = Look At - Bone: dyn_sleeve_tp_l(ComponentPose=@n13, Alpha=1) {
    alphaBoolBlend {blendOption=Linear}
    boneToModify {boneName=dyn_sleeve_tp_l}
    lookAtLocation=<25.3906 -5.0883 13.1524>
    lookAtTarget {boneReference {boneName=upperarm_l}}
    lookAt_Axis {axis=<1 0>}
   }
   n15 = Look At - Bone: dyn_sleeve_bt_l(ComponentPose=@n14, Alpha=1) {
    alphaBoolBlend {blendOption=Linear}
    boneToModify {boneName=dyn_sleeve_bt_l}
    lookAtLocation=<24.5552 0.5885 -11.2935>
    lookAtTarget {boneReference {boneName=upperarm_l}}
    lookAt_Axis {axis=<1 0>}
   }
   n16 = Anim Dynamics - Bone: dyn_sleeve_fr_l!(ComponentPose=@n15, Alpha=1, GravityOverride:Vector="0,0,-980", ExternalForce="0,0,0") -> (Pose) {
    angularDampingOverride=0.9999
    angularSpringConstant=500
    bAngularSpring=true
    bGravityOverrideInSimSpace=true
    bOverrideAngularDamping=true
    bUseGravityOverride=true
    bUseSphericalLimits=true
    simulationSpace=BoneRelative
    alphaBoolBlend {blendOption=Linear}
    boundBone {boneName=dyn_sleeve_fr_l}
    componentLinearAccScale=<0.2 1 1>
    componentLinearVelScale=<0.2 1 1>
    gravityOverride {z=-980}
    physicsBodyDefinitions [{
      sphereCollisionRadius=0
      boundBone {boneName=dyn_sleeve_fr_l}
      boxExtents=<15 30 30>
      constraintSetup {angularLimitsMax=<5 5 15>, angularLimitsMin=<-1 -35 -15>, angularTarget=<0 0 0>}
      localJointOffset=<2 0 0>
     }]
    relativeSpaceBone {boneName=root}
    sphericalLimits [{limitRadius=5, drivingBone {boneName=lowerarm_r}, sphereLocalOffset=<2 -2 0>}]
   }
   n17 = Anim Dynamics - Bone: dyn_sleeve_bk_l!(ComponentPose=@n16, Alpha=1, GravityOverride:Vector="0,0,-980", ExternalForce="0,0,0") -> (Pose) {
    angularDampingOverride=0.9999
    angularSpringConstant=500
    bAngularSpring=true
    bGravityOverrideInSimSpace=true
    bOverrideAngularDamping=true
    bUseGravityOverride=true
    bUseSphericalLimits=true
    simulationSpace=BoneRelative
    alphaBoolBlend {blendOption=Linear}
    boundBone {boneName=dyn_sleeve_bk_l}
    componentLinearAccScale=<0.2 1 1>
    componentLinearVelScale=<0.2 1 1>
    gravityOverride {z=-980}
    physicsBodyDefinitions [{
      sphereCollisionRadius=0
      boundBone {boneName=dyn_sleeve_bk_l}
      boxExtents=<15 30 30>
      constraintSetup {angularLimitsMax=<5 5 15>, angularLimitsMin=<-1 -45 -15>, angularTarget=<0 0 0>}
      localJointOffset=<8 0 0>
     }]
    relativeSpaceBone {boneName=root}
    sphericalLimits [{limitRadius=5, drivingBone {boneName=lowerarm_r}, sphereLocalOffset=<2 -2 0>}]
   }
   n18 = Anim Dynamics - Bone: dyn_sleeve_tp_l!(ComponentPose=@n17, Alpha=1, GravityOverride:Vector="0,0,-980", ExternalForce="0,0,0") -> (Pose) {
    angularDampingOverride=0.9999
    angularSpringConstant=500
    bAngularSpring=true
    bGravityOverrideInSimSpace=true
    bOverrideAngularDamping=true
    bUseGravityOverride=true
    bUseSphericalLimits=true
    simulationSpace=BoneRelative
    alphaBoolBlend {blendOption=Linear}
    boundBone {boneName=dyn_sleeve_tp_l}
    componentLinearAccScale=<0.2 1 1>
    componentLinearVelScale=<0.2 1 1>
    gravityOverride {z=-980}
    physicsBodyDefinitions [{
      sphereCollisionRadius=0
      boundBone {boneName=dyn_sleeve_tp_l}
      boxExtents=<15 30 30>
      constraintSetup {angularLimitsMax=<1 30 15>, angularLimitsMin=<-1 -3 -5>, angularTarget=<0 0 0>}
      localJointOffset=<15 -10 0>
     }]
    relativeSpaceBone {boneName=root}
    sphericalLimits [{limitRadius=5, drivingBone {boneName=lowerarm_r}, sphereLocalOffset=<2 -2 0>}]
   }
   n19 = Anim Dynamics - Bone: dyn_sleeve_bt_l!(ComponentPose=@n18, Alpha=1, GravityOverride:Vector="0,0,-980", ExternalForce="0,0,0") -> (Pose) {
    angularDampingOverride=0.9999
    angularSpringConstant=500
    bAngularSpring=true
    bGravityOverrideInSimSpace=true
    bOverrideAngularDamping=true
    bUseGravityOverride=true
    bUseSphericalLimits=true
    simulationSpace=BoneRelative
    alphaBoolBlend {blendOption=Linear}
    boundBone {boneName=dyn_sleeve_bt_l}
    componentLinearAccScale=<0.2 1 1>
    componentLinearVelScale=<0.2 1 1>
    gravityOverride {z=-980}
    physicsBodyDefinitions [{
      sphereCollisionRadius=0
      boundBone {boneName=dyn_sleeve_bt_l}
      boxExtents=<15 30 30>
      constraintSetup {angularLimitsMax=<5 3 15>, angularLimitsMin=<-1 -45 -15>, angularTarget=<0 0 0>}
      localJointOffset=<15 0 0>
     }]
    relativeSpaceBone {boneName=root}
    sphericalLimits [{limitRadius=5, drivingBone {boneName=lowerarm_r}, sphereLocalOffset=<2 -2 0>}]
   }
   n20 = Look At - Bone: dyn_sleeve_fr_r(ComponentPose=@n19, Alpha=1) {
    alphaBoolBlend {blendOption=Linear}
    boneToModify {boneName=dyn_sleeve_fr_r}
    lookAtLocation=<-34.4234 -18.4331 2.7268>
    lookAtTarget {boneReference {boneName=upperarm_r}}
    lookAt_Axis {axis=<1 0>}
   }
   n21 = Look At - Bone: dyn_sleeve_bk_r(ComponentPose=@n20, Alpha=1) {
    alphaBoolBlend {blendOption=Linear}
    boneToModify {boneName=dyn_sleeve_bk_r}
    lookAtLocation=<-36.3066 17.775 -1.6876>
    lookAtTarget {boneReference {boneName=upperarm_r}}
    lookAt_Axis {axis=<1 0>}
   }
   n22 = Look At - Bone: dyn_sleeve_tp_r(ComponentPose=@n21, Alpha=1) {
    alphaBoolBlend {blendOption=Linear}
    boneToModify {boneName=dyn_sleeve_tp_r}
    lookAtLocation=<-39.5242 -23.9159 -20.9872>
    lookAtTarget {boneReference {boneName=upperarm_r}}
    lookAt_Axis {axis=<1 0>}
   }
   n23 = Look At - Bone: dyn_sleeve_bt_r(ComponentPose=@n22, Alpha=1) {
    alphaBoolBlend {blendOption=Linear}
    boneToModify {boneName=dyn_sleeve_bt_r}
    lookAtLocation=<-46.4725 13.1049 27.9781>
    lookAtTarget {boneReference {boneName=upperarm_r}}
    lookAt_Axis {axis=<1 0>}
   }
   n24 = Anim Dynamics - Bone: dyn_sleeve_fr_r!(ComponentPose=@n23, Alpha=1, GravityOverride:Vector="0,0,-980", ExternalForce="0,0,0") -> (Pose) {
    angularDampingOverride=0.9999
    angularSpringConstant=500
    bAngularSpring=true
    bGravityOverrideInSimSpace=true
    bOverrideAngularDamping=true
    bUseGravityOverride=true
    bUseSphericalLimits=true
    simulationSpace=BoneRelative
    alphaBoolBlend {blendOption=Linear}
    boundBone {boneName=dyn_sleeve_fr_r}
    componentLinearAccScale=<0.2 1 1>
    componentLinearVelScale=<0.2 1 1>
    gravityOverride {z=-980}
    physicsBodyDefinitions [{
      sphereCollisionRadius=0
      boundBone {boneName=dyn_sleeve_fr_r}
      boxExtents=<15 30 30>
      constraintSetup {angularLimitsMax=<5 5 15>, angularLimitsMin=<-1 -35 -15>, angularTarget=<0 0 0>}
      localJointOffset=<2 0 0>
     }]
    relativeSpaceBone {boneName=root}
    sphericalLimits [{limitRadius=5, drivingBone {boneName=lowerarm_r}, sphereLocalOffset=<2 -2 0>}]
   }
   n25 = Anim Dynamics - Bone: dyn_sleeve_bk_r!(ComponentPose=@n24, Alpha=1, GravityOverride:Vector="0,0,-980", ExternalForce="0,0,0") -> (Pose) {
    angularDampingOverride=0.9999
    angularSpringConstant=500
    bAngularSpring=true
    bGravityOverrideInSimSpace=true
    bOverrideAngularDamping=true
    bUseGravityOverride=true
    bUseSphericalLimits=true
    simulationSpace=BoneRelative
    alphaBoolBlend {blendOption=Linear}
    boundBone {boneName=dyn_sleeve_bk_r}
    componentLinearAccScale=<0.2 1 1>
    componentLinearVelScale=<0.2 1 1>
    gravityOverride {z=-980}
    physicsBodyDefinitions [{
      sphereCollisionRadius=0
      boundBone {boneName=dyn_sleeve_bk_r}
      boxExtents=<15 30 30>
      constraintSetup {angularLimitsMax=<5 5 15>, angularLimitsMin=<-1 -45 -15>, angularTarget=<0 0 0>}
      localJointOffset=<8 0 0>
     }]
    relativeSpaceBone {boneName=root}
    sphericalLimits [{limitRadius=5, drivingBone {boneName=lowerarm_r}, sphereLocalOffset=<2 -2 0>}]
   }
   n26 = Anim Dynamics - Bone: dyn_sleeve_tp_r!(ComponentPose=@n25, Alpha=1, GravityOverride:Vector="0,0,-980", ExternalForce="0,0,0") -> (Pose) {
    angularDampingOverride=0.9999
    angularSpringConstant=500
    bAngularSpring=true
    bGravityOverrideInSimSpace=true
    bOverrideAngularDamping=true
    bUseGravityOverride=true
    bUseSphericalLimits=true
    simulationSpace=BoneRelative
    alphaBoolBlend {blendOption=Linear}
    boundBone {boneName=dyn_sleeve_tp_r}
    componentLinearAccScale=<0.2 1 1>
    componentLinearVelScale=<0.2 1 1>
    gravityOverride {z=-980}
    physicsBodyDefinitions [{
      sphereCollisionRadius=0
      boundBone {boneName=dyn_sleeve_tp_r}
      boxExtents=<15 30 30>
      constraintSetup {angularLimitsMax=<1 30 15>, angularLimitsMin=<-1 -3 -5>, angularTarget=<0 0 0>}
      localJointOffset=<15 -10 0>
     }]
    relativeSpaceBone {boneName=root}
    sphericalLimits [{limitRadius=5, drivingBone {boneName=lowerarm_r}, sphereLocalOffset=<2 -2 0>}]
   }
   n27 = Anim Dynamics - Bone: dyn_sleeve_bt_r!(ComponentPose=@n26, Alpha=1, GravityOverride:Vector="0,0,-980", ExternalForce="0,0,0") -> (Pose) {
    angularDampingOverride=0.9999
    angularSpringConstant=500
    bAngularSpring=true
    bGravityOverrideInSimSpace=true
    bOverrideAngularDamping=true
    bUseGravityOverride=true
    bUseSphericalLimits=true
    simulationSpace=BoneRelative
    alphaBoolBlend {blendOption=Linear}
    boundBone {boneName=dyn_sleeve_bt_r}
    componentLinearAccScale=<0.2 1 1>
    componentLinearVelScale=<0.2 1 1>
    gravityOverride {z=-980}
    physicsBodyDefinitions [{
      sphereCollisionRadius=0
      boundBone {boneName=dyn_sleeve_bt_r}
      boxExtents=<15 30 30>
      constraintSetup {angularLimitsMax=<5 3 15>, angularLimitsMin=<-1 -45 -15>, angularTarget=<0 0 0>}
      localJointOffset=<15 0 0>
     }]
    relativeSpaceBone {boneName=root}
    sphericalLimits [{limitRadius=5, drivingBone {boneName=lowerarm_r}, sphereLocalOffset=<2 -2 0>}]
   }
   n28 = "Transform (Modify) Bone - Bone: dyn_sleeve_fr_r"!(ComponentPose=@n27, Alpha=1, Scale="-0.2,1.000,1.000") -> (Pose) {
    rotationMode=BMM_Replace
    rotationSpace=BCS_BoneSpace
    scaleMode=BMM_Replace
    scaleSpace=BCS_BoneSpace
    translationMode=BMM_Additive
    translationSpace=BCS_BoneSpace
    alphaBoolBlend {blendOption=Linear}
    alphaScaleBiasClamp {bInterpResult=true, interpSpeedDecreasing=5, interpSpeedIncreasing=7}
    boneToModify {boneName=dyn_sleeve_fr_r}
    rotation=<rot -0.0002 15.0001 0>
    scale {x=-0.2}
    translation=<1.474 13.2248 -0.0001>
   }
   n29 = "Transform (Modify) Bone - Bone: dyn_sleeve_fr_l"!(ComponentPose=@n28, Alpha=1, Rotation="-5.022,34.989,0.018") -> (Pose) {
    rotationMode=BMM_Replace
    rotationSpace=BCS_BoneSpace
    scaleMode=BMM_Replace
    scaleSpace=BCS_BoneSpace
    translationMode=BMM_Additive
    translationSpace=BCS_BoneSpace
    alphaBoolBlend {blendOption=Linear}
    alphaScaleBiasClamp {bInterpResult=true, interpSpeedDecreasing=5, interpSpeedIncreasing=7}
    boneToModify {boneName=dyn_sleeve_fr_l}
    rotation=<rot -5.022 34.989 0.018>
    translation=<-1.6107 1.1677 0.0015>
   }
   n30 = "Transform (Modify) Bone - Bone: dyn_sleeve_tp_l"!(ComponentPose=@n29, Alpha=1, Rotation="27.010,38.598,9.597") -> (Pose) {
    rotationMode=BMM_Replace
    rotationSpace=BCS_BoneSpace
    scaleMode=BMM_Replace
    scaleSpace=BCS_BoneSpace
    translationMode=BMM_Additive
    translationSpace=BCS_BoneSpace
    alphaBoolBlend {blendOption=Linear}
    alphaScaleBiasClamp {bInterpResult=true, interpSpeedDecreasing=5, interpSpeedIncreasing=7}
    boneToModify {boneName=dyn_sleeve_tp_l}
    rotation=<rot 27.01 38.598 9.597>
    translation=<-2.0201 1.0416 1.1221>
   }
   n31 = "Transform (Modify) Bone - Bone: dyn_sleeve_tp_l"(ComponentPose=@n30, Alpha=1, Rotation="14.978,35.125,-0.025") {
    rotationMode=BMM_Replace
    rotationSpace=BCS_BoneSpace
    scaleMode=BMM_Replace
    scaleSpace=BCS_BoneSpace
    translationMode=BMM_Additive
    translationSpace=BCS_BoneSpace
    alphaBoolBlend {blendOption=Linear}
    alphaScaleBiasClamp {bInterpResult=true, interpSpeedDecreasing=5, interpSpeedIncreasing=7}
    boneToModify {boneName=dyn_sleeve_tp_l}
    rotation=<rot 14.978 35.125 -0.025>
    translation=<-3.8532 2.818 2.9057>
   }
   n32 = "Transform (Modify) Bone - Bone: dyn_sleeve_bk_l"!(ComponentPose=@n31, Alpha=1, Rotation="34.878,-6.213,-3.105") -> (Pose) {
    rotationMode=BMM_Replace
    rotationSpace=BCS_BoneSpace
    scaleMode=BMM_Replace
    scaleSpace=BCS_BoneSpace
    translationMode=BMM_Additive
    translationSpace=BCS_BoneSpace
    alphaBoolBlend {blendOption=Linear}
    alphaScaleBiasClamp {bInterpResult=true, interpSpeedDecreasing=5, interpSpeedIncreasing=7}
    boneToModify {boneName=dyn_sleeve_bk_l}
    rotation=<rot 34.878 -6.213 -3.105>
    translation=<1.498 -4.3653 2.2316>
   }
   n33 = "Constraint - dyn_necklace_01 (PP)"!(ComponentPose=@n32, Alpha=0.75, ConstraintWeights_0=0.7, ConstraintWeights_1:float=0.3) -> (Pose) {
    alphaBoolBlend {blendOption=Linear}
    boneToModify {boneName=dyn_necklace_01}
    constraintSetup [{targetBone {boneName=spine_05}}, {targetBone {boneName=neck_01}}]
    constraintWeights [1,1]
   }
   n34 = Anim Dynamics - Bone: dyn_necklace_01(ComponentPose=@n33, Alpha=1, ExternalForce="0,0,0") {
    angularDampingOverride=0.9999
    angularSpringConstant=500
    bAngularSpring=true
    bGravityOverrideInSimSpace=true
    bOverrideAngularDamping=true
    bUseGravityOverride=true
    bUseSphericalLimits=true
    simulationSpace=BoneRelative
    alphaBoolBlend {blendOption=Linear}
    boundBone {boneName=dyn_necklace_01}
    componentAppliedLinearAccClamp=<3500 3500 3500>
    componentLinearAccScale=<1.2 1.2 1.2>
    componentLinearVelScale=<1.2 1.2 1.2>
    gravityOverride {z=-980}
    physicsBodyDefinitions [{
      sphereCollisionRadius=0
      boundBone {boneName=dyn_necklace_01}
      boxExtents=<5 25 25>
      constraintSetup {angularLimitsMax=<0.5 0.5 3>, angularLimitsMin=<-0.5 -0.5 -3>, angularTarget=<0 0 0>}
      localJointOffset=<2 0 0>
     }]
    relativeSpaceBone {boneName=root}
    sphericalLimits [{limitRadius=2, drivingBone {boneName=spine_05}, sphereLocalOffset=<3 0 0>}]
   }
   n35 = Anim Dynamics - Bone: dyn_necklace_02(ComponentPose=@n34, Alpha=1, ExternalForce="0,0,0") {
    angularDampingOverride=0.9999
    angularSpringConstant=500
    bAngularSpring=true
    bGravityOverrideInSimSpace=true
    bOverrideAngularDamping=true
    bUseGravityOverride=true
    bUseSphericalLimits=true
    simulationSpace=BoneRelative
    alphaBoolBlend {blendOption=Linear}
    boundBone {boneName=dyn_necklace_02}
    componentAppliedLinearAccClamp=<3500 3500 3500>
    componentLinearAccScale=<1.2 1.2 1.2>
    componentLinearVelScale=<1.2 1.2 1.2>
    gravityOverride {z=-980}
    physicsBodyDefinitions [{
      sphereCollisionRadius=0
      boundBone {boneName=dyn_necklace_02}
      boxExtents=<5 25 25>
      constraintSetup {angularLimitsMax=<0.5 0.5 3>, angularLimitsMin=<-0.5 -0.5 -3>, angularTarget=<0 0 0>}
      localJointOffset=<2 0 0>
     }]
    relativeSpaceBone {boneName=root}
    sphericalLimits [{limitRadius=2, drivingBone {boneName=spine_05}, sphereLocalOffset=<-2 0 0>}]
   }
   n36 = "Constraint - dyn_necklace_01 (PPPPPPP)"!(
    ComponentPose=@n35
    Alpha=1
    ConstraintWeights_0=0.0092
    ConstraintWeights_1:float=0.7082
    ConstraintWeights_2:float=0.0019
    ConstraintWeights_3:float=0.002
    ConstraintWeights_4:float=0.0677
    ConstraintWeights_5:float=0.0461
    ConstraintWeights_6:float=0.1648
   ) -> (Pose) {
    alphaBoolBlend {blendOption=Linear}
    alphaScaleBiasClamp {bInterpResult=true, interpSpeedDecreasing=5, interpSpeedIncreasing=7}
    boneToModify {boneName=dyn_necklace_01}
    constraintSetup [
     {targetBone {boneName=spine_04}}
     {targetBone {boneName=spine_05}}
     {targetBone {boneName=clavicle_l}}
     {targetBone {boneName=clavicle_r}}
     {targetBone {boneName=neck_01}}
     {targetBone {boneName=head}}
     {targetBone {boneName=neck_02}}
    ]
    constraintWeights [1,1,1,1,1,1,1]
   }
   n37 = "Constraint - dyn_necklace_02 (PPPPPPP)"!(
    ComponentPose=@n36
    Alpha=1
    ConstraintWeights_0=0.0287
    ConstraintWeights_1:float=0.385
    ConstraintWeights_2:float=0.4491
    ConstraintWeights_3:float=0.0125
    ConstraintWeights_4:float=0.1038
    ConstraintWeights_5:float=0.0192
    ConstraintWeights_6:float=0.0018
   ) -> (Pose) {
    alphaBoolBlend {blendOption=Linear}
    alphaScaleBiasClamp {bInterpResult=true, interpSpeedDecreasing=5, interpSpeedIncreasing=7}
    boneToModify {boneName=dyn_necklace_02}
    constraintSetup [
     {targetBone {boneName=spine_02}}
     {targetBone {boneName=spine_04}}
     {targetBone {boneName=spine_05}}
     {targetBone {boneName=clavicle_l}}
     {targetBone {boneName=spine_03}}
     {targetBone {boneName=clavicle_r}}
     {targetBone {boneName=deform_pec_r}}
    ]
    constraintWeights [1,1,1,1,1,1,1]
   }
   n38 = Anim Dynamics - Bone: dyn_bracelet_01(ComponentPose=@n37, Alpha=1, ExternalForce="0,0,0") {
    angularDampingOverride=1
    angularSpringConstant=20000
    bOverrideLinearDamping=true
    bUseGravityOverride=true
    gravityScale=0.35
    linearDampingOverride=1
    linearSpringConstant=100
    simulationSpace=RootRelative
    alphaBoolBlend {blendOption=Linear}
    boundBone {boneName=dyn_bracelet_01}
    gravityOverride {z=-980}
    physicsBodyDefinitions [{
      boundBone {boneName=dyn_bracelet_01}
      boxExtents=<100 100 100>
      constraintSetup {angularLimitsMax=<0 0.2 0>, angularLimitsMin=<0 -0.2 0>, angularTarget=<0 0 0>, linearAxesMax=<1 0 0>, linearAxesMin=<-0.95 0 0>}
      localJointOffset=<1 0 0>
     }]
    relativeSpaceBone {boneName=upperarm_r}
   }
   n39 = Anim Dynamics - Bone: dyn_can(ComponentPose=@n38, Alpha=1, ExternalForce="0,0,0") {
    angularDampingOverride=1
    angularSpringConstant=1500
    bAngularSpring=true
    bGravityOverrideInSimSpace=true
    bOverrideAngularDamping=true
    gravityScale=0.35
    simulationSpace=RootRelative
    alphaBoolBlend {blendOption=Linear}
    boundBone {boneName=dyn_can}
    physicsBodyDefinitions [{
      sphereCollisionRadius=0
      boundBone {boneName=dyn_can}
      boxExtents=<20 20 5>
      constraintSetup {angularLimitsMax=<10 5 0>, angularLimitsMin=<-5 -3 -5>, angularTarget=<0 0 0>}
      localJointOffset=<0 0 -10>
     }]
   }
   n40 = Anim Dynamics - Bone: dyn_tassel_01_01(ComponentPose=@n39, Alpha=1, ExternalForce="0,0,0") {
    angularSpringConstant=100
    bAngularSpring=true
    bUseGravityOverride=true
    simulationSpace=RootRelative
    alphaBoolBlend {blendOption=Linear}
    boundBone {boneName=dyn_tassel_01_01}
    componentAppliedLinearAccClamp=<10000 10000 10000>
    componentLinearAccScale=<0.1 0.1 0.3>
    componentLinearVelScale=<0.1 0.1 0.3>
    gravityOverride {z=-980}
    physicsBodyDefinitions [{
      boundBone {boneName=dyn_tassel_01_01}
      boxExtents=<10 10 20>
      constraintSetup {angularLimitsMax=<0 70 70>, angularLimitsMin=<0 -70 -100>, angularTarget=<0 0 0>}
      localJointOffset=<1 0 0>
     }]
    relativeSpaceBone {boneName=lowerarm_twist_01_l}
   }
   n41 = Anim Dynamics - Bone: dyn_tassel_02_01(ComponentPose=@n40, Alpha=1, ExternalForce="0,0,0") {
    angularSpringConstant=100
    bAngularSpring=true
    bUseGravityOverride=true
    simulationSpace=RootRelative
    alphaBoolBlend {blendOption=Linear}
    boundBone {boneName=dyn_tassel_02_01}
    componentAppliedLinearAccClamp=<10000 10000 10000>
    componentLinearAccScale=<0.1 0.1 0.3>
    componentLinearVelScale=<0.1 0.1 0.3>
    gravityOverride {z=-980}
    physicsBodyDefinitions [{
      boundBone {boneName=dyn_tassel_02_01}
      boxExtents=<10 10 20>
      constraintSetup {angularLimitsMax=<0 70 70>, angularLimitsMin=<0 -70 -100>, angularTarget=<0 0 0>}
      localJointOffset=<1 0 0>
     }]
    relativeSpaceBone {boneName=lowerarm_twist_01_l}
   }
   n42 = Trail controller - Bone: dyn_tassel_01_04(ComponentPose=@n41, Alpha=1) {
    bUsePlanarLimit=true
    chainLength=3
    relaxationSpeedScale=4
    planarLimits [{drivingBone {boneName=lowerarm_twist_01_l}, planeTransform {location=<0 -3 0>, rotation=<rot 0 0 -110>, scale=<0.3 0.3 0.3>}}]
    rotationLimits [{limitMax=<180 180 180>}, {limitMax=<180 180 180>}, {limitMax=<180 180 180>}]
    rotationOffsets [<0 0 0>, <0 0 0>, <0 0 0>]
    trailBone {boneName=dyn_tassel_01_04}
   }
   n43 = Trail controller - Bone: dyn_tassel_02_04(ComponentPose=@n42, Alpha=1) {
    bUsePlanarLimit=true
    chainLength=3
    relaxationSpeedScale=3
    planarLimits [{drivingBone {boneName=lowerarm_twist_01_l}, planeTransform {location=<0 -3 0>, rotation=<rot 0 0 -110>, scale=<0.3 0.3 0.3>}}]
    rotationLimits [{limitMax=<180 180 180>}, {limitMax=<180 180 180>}, {limitMax=<180 180 180>}]
    rotationOffsets [<0 0 0>, <0 0 0>, <0 0 0>]
    trailBone {boneName=dyn_tassel_02_04}
   }
   n44 = "Constraint - dyn_chain_02 (PPPPPPPPPPPPPP)"!(
    ComponentPose=@n43
    Alpha=1
    ConstraintWeights_0=0.3582
    ConstraintWeights_1:float=0.0001
    ConstraintWeights_2:float=0.0001
    ConstraintWeights_3:float=0.0001
    ConstraintWeights_4:float=0
    ConstraintWeights_5:float=0
    ConstraintWeights_6:float=0.0031
    ConstraintWeights_7:float=0.0004
    ConstraintWeights_8:float=0.0017
    ConstraintWeights_9:float=0.0336
    ConstraintWeights_10:float=0
    ConstraintWeights_11:float=0.4953
    ConstraintWeights_12:float=0.0089
    ConstraintWeights_13:float=0.0985
   ) -> (Pose) {
    alphaBoolBlend {blendOption=Linear}
    alphaScaleBiasClamp {bInterpResult=true, interpSpeedDecreasing=5, interpSpeedIncreasing=7}
    boneToModify {boneName=dyn_chain_02}
    constraintSetup [
     {targetBone {boneName=pelvis}}
     {targetBone {boneName=spine_01}}
     {targetBone {boneName=spine_02}}
     {targetBone {boneName=spine_03}}
     {targetBone {boneName=thigh_l}}
     {targetBone {boneName=calf_l}}
     {targetBone {boneName=thigh_twist_01_l}}
     {targetBone {boneName=deform_glute_l}}
     {targetBone {boneName=deform_groin_l}}
     {targetBone {boneName=thigh_r}}
     {targetBone {boneName=calf_r}}
     {targetBone {boneName=thigh_twist_01_r}}
     {targetBone {boneName=deform_glute_r}}
     {targetBone {boneName=deform_groin_r}}
    ]
    constraintWeights [1,1,1,1,1,1,1,1,1,1,1,1,1,1]
   }
   n45 = "Transform (Modify) Bone - Bone: dyn_chain_MSTR"!(ComponentPose=@n44, Alpha=1, Translation="-1,0,1.5") -> (Pose) {
    rotationMode=BMM_Replace
    rotationSpace=BCS_BoneSpace
    scaleMode=BMM_Replace
    scaleSpace=BCS_BoneSpace
    translationMode=BMM_Replace
    translationSpace=BCS_BoneSpace
    alphaBoolBlend {blendOption=Linear}
    alphaScaleBiasClamp {bInterpResult=true, interpSpeedDecreasing=5, interpSpeedIncreasing=7}
    boneToModify {boneName=dyn_chain_MSTR}
    translation {x=-1, z=1.5}
   }
   n46 = "Transform (Modify) Bone - Bone: dyn_chain_04"(ComponentPose=@n45, Alpha=1, Translation="0,0,0", Rotation="5.000,0,0") {
    rotationMode=BMM_Replace
    rotationSpace=BCS_BoneSpace
    scaleMode=BMM_Replace
    scaleSpace=BCS_BoneSpace
    translationMode=BMM_Additive
    translationSpace=BCS_BoneSpace
    alphaBoolBlend {blendOption=Linear}
    alphaScaleBiasClamp {bInterpResult=true, interpSpeedDecreasing=5, interpSpeedIncreasing=7}
    boneToModify {boneName=dyn_chain_04}
    rotation {pitch=5}
   }
   n47 = "Constraint - dyn_chain_02 (P)"(ComponentPose=@n46, Alpha=1, ConstraintWeights_0=1) {
    alphaBoolBlend {blendOption=Linear}
    boneToModify {boneName=dyn_chain_02}
    constraintSetup [{targetBone {boneName=dyn_chain_01}}]
    constraintWeights [1]
   }
   n48 = Control Rig!(
    Source=Component To Local(ComponentPose=@n47)
    Alpha
    Crouch:bool=Get bIsCrouching()
    CrouchWalk:bool=Get Crouch_move() -> (Crouch_move:bool)
    Tac Sprint:bool=Get bIsTacticalSprinting()
    Jumping:bool=Get JumpingFalling() -> (JumpingFalling:bool)
    Skydive:bool=Get bIsSkyDiving()
   ) -> (Pose) {
    alphaBoolBlend {blendOption=Linear}
    controlRigAssetReference {blueprintRigClass=&/M_MED_SunTide_CtrlRig.M_MED_SunTide_CtrlRig_C}
   }
   n49 = "Constraint - dyn_shorts_bk_l (P)"(ComponentPose=Local To Component(LocalPose=@n48), Alpha=0.8, ConstraintWeights_0=1) {
    alphaBoolBlend {blendOption=Linear}
    boneToModify {boneName=dyn_shorts_bk_l}
    constraintSetup [{targetBone {boneName=calf_l}}]
    constraintWeights [1]
   }
   n50 = "Constraint - dyn_shorts_out_l (P)"(ComponentPose=@n49, Alpha=0.3, ConstraintWeights_0=1) {
    alphaBoolBlend {blendOption=Linear}
    boneToModify {boneName=dyn_shorts_out_l}
    constraintSetup [{targetBone {boneName=dyn_shorts_bk_l}}]
    constraintWeights [1]
   }
   n51 = "Constraint - dyn_shorts_in_l (P)"(ComponentPose=@n50, Alpha=0.3, ConstraintWeights_0=1) {
    alphaBoolBlend {blendOption=Linear}
    boneToModify {boneName=dyn_shorts_in_l}
    constraintSetup [{targetBone {boneName=dyn_shorts_bk_l}}]
    constraintWeights [1]
   }
   n52 = "Constraint - dyn_shorts_bk_r (P)"(ComponentPose=@n51, Alpha=0.8, ConstraintWeights_0=1) {
    alphaBoolBlend {blendOption=Linear}
    boneToModify {boneName=dyn_shorts_bk_r}
    constraintSetup [{targetBone {boneName=calf_r}}]
    constraintWeights [1]
   }
   n53 = "Constraint - dyn_shorts_out_r (P)"(ComponentPose=@n52, Alpha=0.3, ConstraintWeights_0=1) {
    alphaBoolBlend {blendOption=Linear}
    boneToModify {boneName=dyn_shorts_out_r}
    constraintSetup [{targetBone {boneName=dyn_shorts_bk_r}}]
    constraintWeights [1]
   }
   n54 = "Constraint - dyn_shorts_in_r (P)"(ComponentPose=@n53, Alpha=0.3, ConstraintWeights_0=1) {
    alphaBoolBlend {blendOption=Linear}
    boneToModify {boneName=dyn_shorts_in_r}
    constraintSetup [{targetBone {boneName=dyn_shorts_bk_r}}]
    constraintWeights [1]
   }
   n55 = "Constraint - dyn_chain_02 (P)"(ComponentPose=@n54, Alpha=1, ConstraintWeights_0=1) {
    alphaBoolBlend {blendOption=Linear}
    alphaScaleBiasClamp {bInterpResult=true, interpSpeedDecreasing=8}
    boneToModify {boneName=dyn_chain_02}
    constraintSetup [{targetBone {boneName=dyn_chain_01}}]
    constraintWeights [1]
   }
   n56 = "Transform (Modify) Bone - Bone: dyn_chain_MSTR"!(ComponentPose=@n55, Alpha=1, Translation="-1.763,-1.740,-3.894", Rotation:Rotator="1.006,2.959,-28.183") -> (Pose) {
    rotationMode=BMM_Replace
    rotationSpace=BCS_BoneSpace
    scaleMode=BMM_Replace
    scaleSpace=BCS_BoneSpace
    translationMode=BMM_Additive
    translationSpace=BCS_BoneSpace
    alphaBoolBlend {blendOption=Linear}
    alphaScaleBiasClamp {bInterpResult=true, interpSpeedDecreasing=5, interpSpeedIncreasing=7}
    boneToModify {boneName=dyn_chain_MSTR}
    rotation=<rot 1.006 2.959 -28.183>
    translation=<-1.763 -1.74 -3.894>
   }
   n57 = "Transform (Modify) Bone - Bone: dyn_chain_04"(ComponentPose=@n56, Alpha=1, Translation="0,0,0", Rotation="19.254,-5.515,-16.440") {
    rotationMode=BMM_Replace
    rotationSpace=BCS_BoneSpace
    scaleMode=BMM_Replace
    scaleSpace=BCS_BoneSpace
    translationMode=BMM_Additive
    translationSpace=BCS_BoneSpace
    alphaBoolBlend {blendOption=Linear}
    alphaScaleBiasClamp {bInterpResult=true, interpSpeedDecreasing=5, interpSpeedIncreasing=7}
    boneToModify {boneName=dyn_chain_04}
    rotation=<rot 19.254 -5.515 -16.44>
   }
   n58 = "Transform (Modify) Bone - Bone: dyn_chain_MSTR"(ComponentPose=@n57, Alpha=1, Translation="1.834,1.215,-2.807", Rotation="10.000,-0.008,-0.038") {
    rotationMode=BMM_Replace
    rotationSpace=BCS_BoneSpace
    scaleMode=BMM_Replace
    scaleSpace=BCS_BoneSpace
    translationMode=BMM_Additive
    translationSpace=BCS_BoneSpace
    alphaBoolBlend {blendOption=Linear}
    alphaScaleBiasClamp {bInterpResult=true, interpSpeedDecreasing=5, interpSpeedIncreasing=7}
    boneToModify {boneName=dyn_chain_MSTR}
    rotation=<rot 10 -0.008 -0.038>
    translation=<1.834 1.215 -2.807>
   }
   n59 = "Transform (Modify) Bone - Bone: dyn_chain_MSTR"(ComponentPose=@n58, Alpha=1, Translation="0,0,2.425", Rotation="0,0,0") {
    rotationMode=BMM_Replace
    rotationSpace=BCS_BoneSpace
    scaleMode=BMM_Replace
    scaleSpace=BCS_BoneSpace
    translationMode=BMM_Additive
    translationSpace=BCS_BoneSpace
    alphaBoolBlend {blendOption=Linear}
    alphaScaleBiasClamp {bInterpResult=true, interpSpeedDecreasing=5, interpSpeedIncreasing=7}
    boneToModify {boneName=dyn_chain_MSTR}
    translation {z=2.425}
   }
   n60 = "Transform (Modify) Bone - Bone: dyn_chain_MSTR"(ComponentPose=@n59, Alpha=1, Translation="1,0,-0.5", Rotation="0,0,0") {
    rotationMode=BMM_Replace
    rotationSpace=BCS_BoneSpace
    scaleMode=BMM_Replace
    scaleSpace=BCS_BoneSpace
    translationMode=BMM_Additive
    translationSpace=BCS_BoneSpace
    alphaBoolBlend {blendOption=Linear}
    alphaScaleBiasClamp {bInterpResult=true, interpSpeedDecreasing=5, interpSpeedIncreasing=7}
    boneToModify {boneName=dyn_chain_MSTR}
    rotation=<rot 0 0 0>
    translation {x=1, z=-0.5}
   }
   n61 = "Transform (Modify) Bone - Bone: dyn_chain_04"(ComponentPose=@n60, Alpha=1, Translation="0,0,0", Rotation="15,0,0") {
    rotationMode=BMM_Replace
    rotationSpace=BCS_BoneSpace
    scaleMode=BMM_Replace
    scaleSpace=BCS_BoneSpace
    translationMode=BMM_Additive
    translationSpace=BCS_BoneSpace
    alphaBoolBlend {blendOption=Linear}
    alphaScaleBiasClamp {bInterpResult=true, interpSpeedDecreasing=5, interpSpeedIncreasing=7}
    boneToModify {boneName=dyn_chain_04}
    rotation {pitch=15}
   }
   n62 = "Constraint - dyn_chain_02 (P)"(ComponentPose=@n61, Alpha=1, ConstraintWeights_0=1) {
    alphaBoolBlend {blendOption=Linear}
    alphaScaleBiasClamp {bInterpResult=true, interpSpeedDecreasing=5, interpSpeedIncreasing=7}
    boneToModify {boneName=dyn_chain_02}
    constraintSetup [{targetBone {boneName=dyn_chain_01}}]
    constraintWeights [1]
   }
   n63 = "Transform (Modify) Bone - Bone: dyn_chain_MSTR"(ComponentPose=@n62, Alpha=1, Translation="2,-1,1.5", Rotation="0.000,-0.000,-15.5") {
    rotationMode=BMM_Replace
    rotationSpace=BCS_BoneSpace
    scaleMode=BMM_Replace
    scaleSpace=BCS_BoneSpace
    translationMode=BMM_Additive
    translationSpace=BCS_BoneSpace
    alphaBoolBlend {blendOption=Linear}
    alphaScaleBiasClamp {bInterpResult=true, interpSpeedDecreasing=5, interpSpeedIncreasing=7}
    boneToModify {boneName=dyn_chain_MSTR}
    rotation {roll=-15.5}
    translation=<2 -1 1.5>
   }
   n64 = "Constraint - dyn_shorts_bk_l (P)"(ComponentPose=@n63, Alpha=1, ConstraintWeights_0=1) {
    alphaBoolBlend {blendOption=Linear}
    boneToModify {boneName=dyn_shorts_bk_l}
    constraintSetup [{targetBone {boneName=calf_l}}]
    constraintWeights [1]
   }
   n65 = "Constraint - dyn_shorts_out_l (P)"(ComponentPose=@n64, Alpha=1, ConstraintWeights_0=1) {
    alphaBoolBlend {blendOption=Linear}
    boneToModify {boneName=dyn_shorts_out_l}
    constraintSetup [{targetBone {boneName=dyn_shorts_bk_l}}]
    constraintWeights [1]
   }
   n66 = "Constraint - dyn_shorts_in_l (P)"(ComponentPose=@n65, Alpha=1, ConstraintWeights_0=1) {
    alphaBoolBlend {blendOption=Linear}
    boneToModify {boneName=dyn_shorts_in_l}
    constraintSetup [{targetBone {boneName=dyn_shorts_bk_l}}]
    constraintWeights [1]
   }
   n67 = "Constraint - dyn_shorts_bk_r (P)"(ComponentPose=@n66, Alpha=1, ConstraintWeights_0=1) {
    alphaBoolBlend {blendOption=Linear}
    boneToModify {boneName=dyn_shorts_bk_r}
    constraintSetup [{targetBone {boneName=calf_r}}]
    constraintWeights [1]
   }
   n68 = "Constraint - dyn_shorts_out_r (P)"(ComponentPose=@n67, Alpha=1, ConstraintWeights_0=1) {
    alphaBoolBlend {blendOption=Linear}
    boneToModify {boneName=dyn_shorts_out_r}
    constraintSetup [{targetBone {boneName=dyn_shorts_bk_r}}]
    constraintWeights [1]
   }
   n69 = "Constraint - dyn_shorts_in_r (P)"(ComponentPose=@n68, Alpha=1, ConstraintWeights_0=1) {
    alphaBoolBlend {blendOption=Linear}
    boneToModify {boneName=dyn_shorts_in_r}
    constraintSetup [{targetBone {boneName=dyn_shorts_bk_r}}]
    constraintWeights [1]
   }
   n70 = "Constraint - dyn_necklace_02 (PPPPPPP)"(
    ComponentPose=@n69
    Alpha=1
    ConstraintWeights_0=0.0287
    ConstraintWeights_1=0.385
    ConstraintWeights_2=0.4491
    ConstraintWeights_3=0.0125
    ConstraintWeights_4=0.1038
    ConstraintWeights_5=0.0192
    ConstraintWeights_6=0.0018
   ) {
    alphaBoolBlend {blendOption=Linear}
    boneToModify {boneName=dyn_necklace_02}
    constraintSetup [
     {targetBone {boneName=spine_02}}
     {targetBone {boneName=spine_04}}
     {targetBone {boneName=spine_05}}
     {targetBone {boneName=clavicle_l}}
     {targetBone {boneName=spine_03}}
     {targetBone {boneName=clavicle_r}}
     {targetBone {boneName=deform_pec_r}}
    ]
    constraintWeights [1,1,1,1,1,1,1]
   }
   Output Pose(Result=Component To Local(ComponentPose=@n70)) {layerGroup=None}
  ]}
 EventGraph {nodes [
   Event BlueprintUpdateAnimation()
   > n1 = Sequence!(execute) -> (then_0, then_1, then_2:Exec, then_3:Exec)
   > n2 = Gravity Override State Function DEPRECATED(CharacterPartAnimInstance=Self-Reference(), GameTimeInSeconds=GetGameTimeInSeconds(), GravityOverrideParams=Get Gravity Override Params())
   > Set gravity_override(gravity_override=@n2)
   n3 = Branch(execute=@n1.then_1, Condition=OR Boolean(A=Get bIsJumping(), B=Get bIsFalling()))
   > Set JumpingFalling(JumpingFalling:bool=true) -> (Output_Get:bool)
   Set JumpingFalling(execute=@n3.else, JumpingFalling=false)
   n4 = Branch(execute=@n1.then_2, Condition=AND Boolean(A=Get bIsCrouching(), B=Get bIsMoving2D()))
   > Set Crouch_move(Crouch_move:bool=true) -> (Output_Get:bool)
   Set Crouch_move(execute=@n4.else, Crouch_move=false)
   Set Emote(execute=@n1.then_3, Emote:double=SelectFloat(A=1, B=0, bPickA=Get bIsPlayingEmote())) -> (Output_Get:double)
   Get bIsCrouching()
  ]}
]
variables [
 calf_l {type=number}
 calf_r {type=number}
 Crouch_move {type=boolean}
 Emote {type=number}
 Gravity Override Params {type=GravityOverrideParamsStruct}
 gravity_override {type=Vector}
 JumpingFalling {type=boolean}
]
