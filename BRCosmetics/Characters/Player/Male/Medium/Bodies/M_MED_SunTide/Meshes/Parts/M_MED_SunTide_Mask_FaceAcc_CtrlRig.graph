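# Control Rig for the SunTide male medium mask face accessory that adjusts the back hair bone dyn_hair_bk as the head moves.
# Two spherical pose readers measure head orientation against spine_05, each weighting its own offset and rotation on that bone, applied in sequence.
# The second reader has debug drawing enabled.
# A head-to-spine Euler angle remap is also computed, but its result drives nothing.
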ControlRigBlueprint /BRCosmetics/Characters/Player/Male/Medium/Bodies/M_MED_SunTide/Meshes/Parts/M_MED_SunTide_Mask_FaceAcc_CtrlRig
graphs [
 RigVMFunctionLibrary {links [], nodes []}
 RigVMModel {nodes [
   RigUnit_BeginExecution = &RigUnit_BeginExecution()
   n1 = &RigUnit_SphericalPoseReader(
    ActiveRegionScaleFactors=(PositiveWidth=0.5,NegativeWidth=0.5,PositiveHeight=0.5,NegativeHeight=0.5)
    ActiveRegionSize=0.1
    Debug=(bDrawDebug=false,DebugScale=25,DebugSegments=20,DebugThickness=0.25)
    DriverAxis=(X=1,Y=0,Z=0)
    DriverItem=(Type=Bone,Name="head")
    FalloffRegionScaleFactors=(PositiveWidth=1,NegativeWidth=1,PositiveHeight=1.5,NegativeHeight=1)
    FalloffSize=0.2
    OptionalParentItem=(Type=Bone,Name="spine_05")
    RotationOffset=(X=0,Y=-55,Z=-50)
    ExecutePin=@RigUnit_BeginExecution.ExecutePin
   )
   > &RigUnit_ModifyTransforms(
    ItemToModify=((Item=(Type=Bone,Name="dyn_hair_bk"),Transform=(Rotation=(X=0,Y=0,Z=0,W=1),Translation=(X=0,Y=-1,Z=2),Scale3D=(X=1,Y=1,Z=1))))
    ItemToModify.0.Transform.Rotation=&RigVMFunction_MathQuaternionFromEuler(Euler=(X=-90,Y=0,Z=0)).Result
    Weight=@n1.OutputParam
   )
   > n2 = &RigUnit_SphericalPoseReader(
    ActiveRegionScaleFactors=(PositiveWidth=0.5,NegativeWidth=0.5,PositiveHeight=0.5,NegativeHeight=0.5)
    ActiveRegionSize=0.1
    Debug=(bDrawDebug=True,DebugScale=25,DebugSegments=20,DebugThickness=0.25)
    DriverAxis=(X=1,Y=0,Z=0)
    DriverItem=(Type=Bone,Name="head")
    FalloffRegionScaleFactors=(PositiveWidth=1,NegativeWidth=1,PositiveHeight=1,NegativeHeight=1)
    FalloffSize=0.2
    OptionalParentItem=(Type=Bone,Name="spine_05")
    RotationOffset=(X=0,Y=-35,Z=-40)
   )
   > &RigUnit_ModifyTransforms(
    ItemToModify=((Item=(Type=Bone,Name="dyn_hair_bk"),Transform=(Rotation=(X=0,Y=0,Z=0,W=1),Translation=(X=0,Y=-1,Z=1),Scale3D=(X=1,Y=1,Z=1))))
    ItemToModify.0.Transform.Rotation=&RigVMFunction_MathQuaternionFromEuler(Euler=(X=-20,Y=20,Z=0)).Result
    Weight=@n2.OutputParam
   )
   n3 = &RigUnit_GetRelativeTransformForItem(Child=(Type=Bone,Name="head"), Parent=(Type=Bone,Name="spine_05"), bChildInitial=false, bParentInitial=false)
   &RigVMFunction_MathFloatRemap(
    SourceMaximum=-20
    SourceMinimum=-60
    TargetMaximum=0
    TargetMinimum=1
    bClamp=true
    Value=&RigVMFunction_MathFloatAdd(B=0, A=&RigVMFunction_MathQuaternionToEuler(RotationOrder=ZYX, Value=@n3.RelativeTransform.Rotation).Result.Z) {template="Add::Execute(in A,in B,out Result)"}.Result
   )
  ]}
]
hierarchy [
 root {loc=<0 0 0>, rot=<0 0 0 1>}
 pelvis {parent=root, loc=<0 -0.7071 93.1191>, rot=<0.0127 0.707 -0.0127 -0.707>}
 spine_01 {parent=pelvis, loc=<6.7494 -0.6388 0>, rot=<0 0 0.0503 -0.9987>}
 spine_02 {parent=spine_01, loc=<7.2493 -0.0308 0>, rot=<0 0 0.0041 -1>}
 spine_03 {parent=spine_02, loc=<6.0084 -0.0167 0>, rot=<0 0 0.0027 -1>}
 spine_04 {parent=spine_03, loc=<8.0999 -0.0224 0>, rot=<0 0 0.0026 -1>}
 spine_05 {parent=spine_04, loc=<10.0119 -0.0184 0>, rot=<0 0 -0.0016 -1>}
 neck_01 {parent=spine_05, loc=<11.3138 0.1529 0>, rot=<0 0 -0.2003 -0.9797>}
 neck_02 {parent=neck_01, loc=<4.5627 0 0>, rot=<0 0 0.0504 -0.9987>}
 head {parent=neck_02, loc=<5.3914 0 0>, rot=<0 0 0.115 -0.9934>}
 dyn_hairstrand_01 {parent=head, loc=<11.4477 5.8508 5.4108>, rot=<0.1241 0.1253 0.1231 -0.9766>}
 dyn_hairstrand_02 {parent=dyn_hairstrand_01, loc=<4.1416 0 0>, rot=<0 -0.5331 0 -0.8461>}
 dyn_hairstrand_03 {parent=head, loc=<17.332 -0.3692 2.3642>, rot=<0.1043 -0.6687 0.1074 -0.7283>}
 dyn_hairstrand_04 {parent=dyn_hairstrand_03, loc=<6.1956 0 0>, rot=<-0.1163 -0.3908 0.3549 -0.8413>}
 dyn_earring {parent=head, loc=<0.8339 1.5066 6.162>, rot=<0.7071 0.0029 0.7071 -0.0029>}
 dyn_hair_bk {parent=head, loc=<1.0267 -5.1248 0>, rot=<0.0029 -0.7071 -0.0029 -0.7071>}
 dyn_hair_a_mstr {parent=head, loc=<9.9825 10.0697 -5.7249>, rot=<-0.0607 -0.9504 -0.3046 -0.0167>}
 dyn_hair_a_01 {parent=dyn_hair_a_mstr, loc=<0 0 0>, rot=<0 0 0 1>}
 dyn_hair_a_02 {parent=dyn_hair_a_01, loc=<4.8036 0 0>, rot=<0 0 0 1>}
 dyn_hair_b_mstr {parent=head, loc=<8.935 6.7273 -7.6691>, rot=<0.0191 -0.9725 -0.232 0.0084>}
 dyn_hair_b_01 {parent=dyn_hair_b_mstr, loc=<0 0 0>, rot=<0 0 0 1>}
 dyn_hair_b_02 {parent=dyn_hair_b_01, loc=<5.1402 0 0>, rot=<-0.02 -0.0153 -0.0009 -0.9997>}
 dyn_hair_c_mstr {parent=head, loc=<7.3756 2.448 -8.3854>, rot=<0.0682 -0.9977 0 0>}
 dyn_hair_c_01 {parent=dyn_hair_c_mstr, loc=<0 0 0>, rot=<0 0 0 1>}
 dyn_hair_c_02 {parent=dyn_hair_c_01, loc=<2.2634 1.3859 0>, rot=<0 0 -0.0641 -0.9979>}
 dyn_hair_c_03 {parent=dyn_hair_c_01, loc=<2.6785 -1.9465 0.8256>, rot=<0 0 -0.0641 -0.9979>}
 faceAttach {parent=head, loc=<0.6224 0.6563 0>, rot=<0.0029 -0.7071 -0.0029 -0.7071>}
 C_jaw {parent=faceAttach, loc=<-0.0358 1.6717 -1.9384>, rot=<0 0 0 1>}
 clavicle_l {parent=spine_05, loc=<4.3575 -0.1466 -2.3223>, rot=<0.0476 -0.7289 -0.0102 -0.6829>}
 clavicle_r {parent=spine_05, loc=<4.3575 -0.1466 2.3223>, rot=<0.7289 0.0476 -0.6829 0.0102>}
]
